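# Sound cue for wetland POI ambient spot sounds, playing one random distant creature call from deer, wapiti, bullfrog, cricket frog or frog recordings.
# An equally weighted Random picks a species branch, whose Modulator varies playback, with nested Randoms choosing among variants except for the single wapiti wave.
# The frog branch is boosted well above the others in volume.
# All quality levels share the same tree, played through distant-creature attenuation and the outdoor ambient sound class.
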SoundCue /Game/Sounds/Helios/POI/Wetland/Spots/Spot_Wetland_Creatures_All_Cue
root=QualityLevel_0
nodes [
 {id=Modulator_0, children [Random_8], properties {volumeMax=2.5, volumeMin=2}}
 {id=Modulator_4, children [Random_4]}
 {id=Modulator_5, children [WavePlayer_5]}
 {id=Modulator_6, children [Random_5]}
 {id=Modulator_7, children [Random_6], properties {volumeMin=0.7}}
 {id=QualityLevel_0, children [Random_9,Random_9,Random_9,Random_9]}
 {id=Random_4, children [WavePlayer_0,WavePlayer_1], properties {bShouldExcludeFromBranchCulling=true, weights [1,1]}}
 {
  id=Random_5
  children [WavePlayer_6,WavePlayer_7,WavePlayer_8,WavePlayer_9]
  properties {bShouldExcludeFromBranchCulling=true, weights [1,1,1,1]}
 }
 {
  id=Random_6
  children [WavePlayer_10, WavePlayer_11, WavePlayer_12, WavePlayer_13]
  properties {bShouldExcludeFromBranchCulling=true, weights [1,1,1,1]}
 }
 {
  id=Random_8
  children [WavePlayer_14, WavePlayer_15, WavePlayer_16, WavePlayer_17, WavePlayer_18, WavePlayer_19]
  properties {bShouldExcludeFromBranchCulling=true, weights [1,1,1,1,1,1]}
 }
 {id=Random_9, children [Modulator_4,Modulator_5,Modulator_6,Modulator_7,Modulator_0], properties {weights [1,1,1,1,1]}}
 {id=WavePlayer_0, properties {soundWaveAssetPtr=&/Spot_Wetland_Animal_Deer_01}}
 {id=WavePlayer_1, properties {soundWaveAssetPtr=&/Spot_Wetland_Animal_Deer_02}}
 {id=WavePlayer_10, properties {soundWaveAssetPtr=&/Spot_Wetland_CricketFrog_02}}
 {id=WavePlayer_11, properties {soundWaveAssetPtr=&/Spot_Wetland_CricketFrog_01}}
 {id=WavePlayer_12, properties {soundWaveAssetPtr=&/Spot_Wetland_CricketFrog_04}}
 {id=WavePlayer_13, properties {soundWaveAssetPtr=&/Spot_Wetland_CricketFrog_03}}
 {id=WavePlayer_14, properties {soundWaveAssetPtr=&/Spot_Wetland_Frog_01}}
 {id=WavePlayer_15, properties {soundWaveAssetPtr=&/Spot_Wetland_Frog_02}}
 {id=WavePlayer_16, properties {soundWaveAssetPtr=&/Spot_Wetland_Frog_03}}
 {id=WavePlayer_17, properties {soundWaveAssetPtr=&/Spot_Wetland_Frog_04}}
 {id=WavePlayer_18, properties {soundWaveAssetPtr=&/Spot_Wetland_Frog_06}}
 {id=WavePlayer_19, properties {soundWaveAssetPtr=&/Spot_Wetland_Frog_05}}
 {id=WavePlayer_5, properties {soundWaveAssetPtr=&/Spot_Wetland_Animal_Wapiti_01}}
 {id=WavePlayer_6, properties {soundWaveAssetPtr=&/Spot_Wetland_Bullfrog_04}}
 {id=WavePlayer_7, properties {soundWaveAssetPtr=&/Spot_Wetland_Bullfrog_03}}
 {id=WavePlayer_8, properties {soundWaveAssetPtr=&/Spot_Wetland_Bullfrog_02}}
 {id=WavePlayer_9, properties {soundWaveAssetPtr=&/Spot_Wetland_Bullfrog_01}}
]
properties {
 attenuationSettings=/Game/Sounds/Attenuation/Fort_TileAmbient_OneShot_Distant_Creature_Att
 bEnableSubmixSends=false
 maxDistance=8500
 soundClassObject=&Fort_Ambient_Outdoors_Athena
 soundSubmixObject=&Fortnite_Reverb_Distant
 volumeMultiplier=0.4
 preEffectBusSends [{maxSendDistance=8000, minSendDistance=3000, minSendLevel=0.6}]
}
