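# Sound cue for distant creature spot ambience in the Helios wetland POI.
# A weighted random favors three Wetland animal one-shots, mixed down in volume, over seven groans reused from Artemis Cattus Cave.
# Both branches randomize pitch, and the groans also vary in volume.
# It uses distant-creature attenuation settings with distance-scaled sends to an exterior ambient bus and distant reverb submix.
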SoundCue /Game/Sounds/Helios/POI/Wetland/Spots/Spot_Wetland_Distant_Creature_Groans_Cue
root=Modulator_0
nodes [
 {id=Mixer_0, children [Modulator_3], properties {inputVolume [0.4]}}
 {id=Modulator_0, children [Random_2]}
 {id=Modulator_1, children [Random_0], properties {pitchMax=1.1, pitchMin=0.75, volumeMax=1, volumeMin=0.8}}
 {id=Modulator_3, children [Random_3], properties {pitchMax=1.1, pitchMin=0.75, volumeMax=1, volumeMin=1}}
 {
  id=Random_0
  children [WavePlayer_0,WavePlayer_1,WavePlayer_2,WavePlayer_3,WavePlayer_4,WavePlayer_5,WavePlayer_6]
  properties {weights [1,1,1,1,1,1,1]}
 }
 {id=Random_2, children [Mixer_0,Modulator_1], properties {weights [1,0.25]}}
 {id=Random_3, children [WavePlayer_7,WavePlayer_8,WavePlayer_9], properties {bShouldExcludeFromBranchCulling=true, weights [1,1,1]}}
 {id=WavePlayer_0, properties {soundWaveAssetPtr=/Game/Sounds/Artemis/POI/Cattus_Cave/Spots/Spot_Cattus_Cave_Distant_Creature_Groan_01}}
 {id=WavePlayer_1, properties {soundWaveAssetPtr=/Game/Sounds/Artemis/POI/Cattus_Cave/Spots/Spot_Cattus_Cave_Distant_Creature_Groan_02}}
 {id=WavePlayer_2, properties {soundWaveAssetPtr=/Game/Sounds/Artemis/POI/Cattus_Cave/Spots/Spot_Cattus_Cave_Distant_Creature_Groan_03}}
 {id=WavePlayer_3, properties {soundWaveAssetPtr=/Game/Sounds/Artemis/POI/Cattus_Cave/Spots/Spot_Cattus_Cave_Distant_Creature_Groan_04}}
 {id=WavePlayer_4, properties {soundWaveAssetPtr=/Game/Sounds/Artemis/POI/Cattus_Cave/Spots/Spot_Cattus_Cave_Distant_Creature_Groan_05}}
 {id=WavePlayer_5, properties {soundWaveAssetPtr=/Game/Sounds/Artemis/POI/Cattus_Cave/Spots/Spot_Cattus_Cave_Distant_Creature_Groan_06}}
 {id=WavePlayer_6, properties {soundWaveAssetPtr=/Game/Sounds/Artemis/POI/Cattus_Cave/Spots/Spot_Cattus_Cave_Distant_Creature_Groan_07}}
 {id=WavePlayer_7, properties {soundWaveAssetPtr=&/Spot_Wetland_Animal_OneShots_01}}
 {id=WavePlayer_8, properties {soundWaveAssetPtr=&/Spot_Wetland_Animal_OneShots_02}}
 {id=WavePlayer_9, properties {soundWaveAssetPtr=&/Spot_Wetland_Animal_OneShots_03}}
]
properties {
 attenuationSettings=/Game/Sounds/Attenuation/Fort_TileAmbient_OneShot_Distant_Creature_Att
 maxDistance=8500
 soundClassObject=&Fort_Ambient_OneShots_Outdoors
 volumeMultiplier=0.5
 attenuationOverrides {bAttenuate=false}
 busSends [{
   audioBus=null
   maxSendDistance=8000
   minSendDistance=3000
   minSendLevel=0.6
   sendLevel=1
   soundSourceBus=&AmbientAudio_Exterior_SourceBus
   customSendLevelCurve {externalCurve=/Game/Sounds/SubMixes/Reverb/Fortnite_Reverb_Distant_Send}
  }]
 soundSubmixSends [{
   maxSendDistance=8000
   maxSendLevel=0.4
   minSendDistance=3000
   minSendLevel=0.25
   soundSubmix=&Fortnite_Reverb_Distant
   customSendLevelCurve {editorCurveData {keys [<at 0 0>,<at 1 1>]}}
  }]
}
