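# Sound cue for wetland insect and fly-by spot ambience, classed as an outdoor ambient one-shot with insect fly-by attenuation.
# A weighted Random node picks evenly among six insect waves and three flyby waves, with the first insect wave boosted through a Mixer.
# A Modulator varies the pick, and every quality level shares that same chain.
# An exterior ambient source-bus send is configured, but bus sends are disabled.
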SoundCue /Game/Sounds/Helios/POI/Wetland/Spots/Spot_Wetland_Fly_Cue
root=QualityLevel_0
nodes [
 {id=Mixer_0, children [WavePlayer_0], properties {inputVolume [2]}}
 {id=Modulator_1, children [Random_1], properties {pitchMax=1.1}}
 {id=QualityLevel_0, children [Modulator_1,Modulator_1,Modulator_1,Modulator_1]}
 {
  id=Random_1
  children [Mixer_0, WavePlayer_1, WavePlayer_2, WavePlayer_3, WavePlayer_8, WavePlayer_9, WavePlayer_10, WavePlayer_11, WavePlayer_12]
  properties {weights [1,1,1,1,1,1,1,1,1]}
 }
 {id=WavePlayer_0, properties {soundWaveAssetPtr=&/Spot_Wetland_Insects_01}}
 {id=WavePlayer_1, properties {soundWaveAssetPtr=&/Spot_Wetland_Insects_02}}
 {id=WavePlayer_10, properties {soundWaveAssetPtr=&/Spot_Wetland_Insects_Flyby_01}}
 {id=WavePlayer_11, properties {soundWaveAssetPtr=&/Spot_Wetland_Insects_Flyby_02}}
 {id=WavePlayer_12, properties {soundWaveAssetPtr=&/Spot_Wetland_Insects_Flyby_03}}
 {id=WavePlayer_2, properties {soundWaveAssetPtr=&/Spot_Wetland_Insects_03}}
 {id=WavePlayer_3, properties {soundWaveAssetPtr=&/Spot_Wetland_Insects_04}}
 {id=WavePlayer_8, properties {soundWaveAssetPtr=&/Spot_Wetland_Insects_05}}
 {id=WavePlayer_9, properties {soundWaveAssetPtr=&/Spot_Wetland_Insects_06}}
]
properties {
 attenuationSettings=/Game/Sounds/Attenuation/Fort_TileAmbient_OneShot_Insect_By_Att
 bEnableBusSends=false
 maxDistance=6000
 soundClassObject=&Fort_Ambient_OneShots_Outdoors
 volumeMultiplier=0.4
 busSends [{
   audioBus=null
   maxSendDistance=6000
   minSendDistance=3000
   minSendLevel=0.6
   sendLevel=1
   soundSourceBus=&AmbientAudio_Exterior_SourceBus
  }]
}
