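# Sound cue for distant one-shot rumble spot ambience in the Helios wetland POI, classed as outdoor ambient audio.
# A Random node picks one of six equally weighted rumble waves and feeds a Modulator that varies volume per play.
# It uses a distant-creature attenuation asset and a concurrency override that caps simultaneous instances at two.
# Distance-ranged sends route it to the exterior ambient source bus and the default reverb submix.
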SoundCue /Game/Sounds/Helios/POI/Wetland/Spots/Spot_Wetland_Rumbles_Cue
root=Modulator_0
nodes [
 {id=Modulator_0, children [Random_1], properties {volumeMax=1, volumeMin=0.8}}
 {id=Random_1, children [WavePlayer_0,WavePlayer_1,WavePlayer_2,WavePlayer_3,WavePlayer_4,WavePlayer_5], properties {weights [1,1,1,1,1,1]}}
 {id=WavePlayer_0, properties {soundWaveAssetPtr=&/Spot_Wetland_Rumble_01}}
 {id=WavePlayer_1, properties {soundWaveAssetPtr=&/Spot_Wetland_Rumble_02}}
 {id=WavePlayer_2, properties {soundWaveAssetPtr=&/Spot_Wetland_Rumble_03}}
 {id=WavePlayer_3, properties {soundWaveAssetPtr=&/Spot_Wetland_Rumble_04}}
 {id=WavePlayer_4, properties {soundWaveAssetPtr=&/Spot_Wetland_Rumble_05}}
 {id=WavePlayer_5, properties {soundWaveAssetPtr=&/Spot_Wetland_Rumble_06}}
]
properties {
 attenuationSettings=/Game/Sounds/Attenuation/Fort_TileAmbient_OneShot_Distant_Creature_Att
 bOverrideConcurrency=true
 maxDistance=8500
 soundClassObject=&Fort_Ambient_Outdoors_Athena
 volumeMultiplier=0.5
 busSends [{
   audioBus=null
   maxSendDistance=8000
   minSendDistance=3000
   minSendLevel=0.4
   sendLevel=1
   soundSourceBus=&AmbientAudio_Exterior_SourceBus
  }]
 concurrencyOverrides {maxCount=2}
 soundSubmixSends [{
   maxSendDistance=8000
   maxSendLevel=0.6
   minSendDistance=3000
   minSendLevel=0.4
   soundSubmix=&Fortnite_Reverb_Default
   customSendLevelCurve {editorCurveData {keys [<at 0 0.4>,<at 1 1>]}}
  }]
}
