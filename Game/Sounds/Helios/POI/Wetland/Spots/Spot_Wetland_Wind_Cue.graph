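# Sound cue for one-shot wind spot ambience at the Helios Wetland POI, in the outdoor ambient one-shots class.
# An equal-weight random pick from five wind waves passes through a modulator that varies pitch and volume.
# The quality-level root routes all four of its inputs to that same modulator.
# A wind attenuation asset, low volume multiplier, concurrency override capping simultaneous voices and distance-based pre-effect bus send shape playback.
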SoundCue /Game/Sounds/Helios/POI/Wetland/Spots/Spot_Wetland_Wind_Cue
root=QualityLevel_0
nodes [
 {id=Modulator_0, children [Random_0], properties {pitchMax=1.4, pitchMin=0.9, volumeMax=1.2}}
 {id=QualityLevel_0, children [Modulator_0,Modulator_0,Modulator_0,Modulator_0]}
 {id=Random_0, children [WavePlayer_0,WavePlayer_1,WavePlayer_2,WavePlayer_3,WavePlayer_4], properties {weights [1,1,1,1,1]}}
 {id=WavePlayer_0, properties {soundWaveAssetPtr=&/Spot_Wetland_Wind_01}}
 {id=WavePlayer_1, properties {soundWaveAssetPtr=&/Spot_Wetland_Wind_02}}
 {id=WavePlayer_2, properties {soundWaveAssetPtr=&/Spot_Wetland_Wind_03}}
 {id=WavePlayer_3, properties {soundWaveAssetPtr=&/Spot_Wetland_Wind_04}}
 {id=WavePlayer_4, properties {soundWaveAssetPtr=&/Spot_Wetland_Wind_05}}
]
properties {
 attenuationSettings=/Game/Sounds/Attenuation/Fort_TileAmbient_OneShot_Wind_Att
 bOverrideConcurrency=true
 maxDistance=8000
 soundClassObject=&Fort_Ambient_OneShots_Outdoors
 volumeMultiplier=0.3
 concurrencyOverrides {maxCount=2, voiceStealReleaseTime=1}
 preEffectBusSends [{maxSendDistance=6000, minSendDistance=3000, minSendLevel=0.6}]
}
